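; Stage event actor, spawned by BP_MojoStageSequence, that plays one FX manager event data asset on the stage's active FX managers.
; StartEvent applies any start delay, then PlayEvent targets all managers or listed IDs, with split-name key matching on battle stages.
; It binds each EventFinished and fires from a local copy, surviving same-frame removals.
; EventComplete fires when none remain or none were found; DestroyEvent clears the timer and bindings.
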(fn CheckIfEventIsComplete
  (params (FxManager :type BP%20Mojo%20Fx%20Manager%20Object%20Reference))
  (bind _activefxmanagers (Variables|Default|GetActiveFXManagers))
  ;[1 1 1] If the FxManager is on the ActiveFxManagers array, remove from it and unbind from the FxManager Event Finished dispatcher
  (if (>= (Utilities|Array|FindItem _activefxmanagers FxManager) 0)
    (Default|UnbindEventfromEventFinished FxManager (EventDispatchers|CreateEvent :!event Check%20if%20Event%20Is%20Complete))
    (Utilities|Array|RemoveItem _activefxmanagers FxManager)
    ;[1 1 1] If there's no remaining active FxManagers, this event is done.
    (if (Utilities|Array|IsEmpty _activefxmanagers)
      ;[1 1 1] The BP_MojoStageSequence that spawned this StageEvent binds the CheckIfAllEventsComplete function to this dispatcher.
      ; That's through here the sequence knows that all events finished
      (Default|CallEventComplete self)
      (if (Variables|/SparksStageShow/Systems/StageManager/BP_MojoStageManagerBase.BP_MojoStageManagerBase_C|GetEnableDebugLogs (Variables|Default|GetStageManager))
        (Development|PrintText (Utilities|Text|FormatText "NSLOCTEXT(\"[4E16E6D6F54585BA172502B426D9D680]\", \"7DB041BF45B99B108F8CBD93F4674AF7\", \"[StageEvent] Event finished: {event}.\")" (Utilities|GetDisplayName (Variables|Default|GetFXManagerEvent))) false)))))

(fn DestroyEvent
  (bind _startdelaytimerhandle (Variables|Default|GetStartDelayTimerHandle))
  (if (Utilities|Time|IsValidTimerHandle _startdelaytimerhandle)
    (Utilities|Time|ClearandInvalidateTimerbyHandle _startdelaytimerhandle)
    (else
      _))
  (for _array_element (Variables|Default|GetActiveFXManagers)
    (Utilities|IsValid _array_element
      (:Is%20Valid
        (Default|UnbindEventfromEventFinished _array_element (EventDispatchers|CreateEvent :!event Check%20If%20Event%20Is%20Complete)))))
  (Actor|DestroyActor))

(fn GetActiveFxManagers
  (params (FXManagers :type Array%20of%20BP%20Mojo%20Fx%20Manager%20Object%20References))
  (Utilities|Array|Clear (LocalVariables|GetLocalActiveFXManagers))
  (for _array_element FXManagers
    (if (Variables|/SparksStageShow/Systems/StageManager/FXManager/BP_MojoFxManager.BP_MojoFxManager_C|GetIsActive _array_element)
      (Utilities|Array|Add (LocalVariables|GetLocalActiveFXManagers) _array_element)))
  (return (LocalVariables|GetLocalActiveFXManagers)))

(fn GetValidFxManagers
  (params (FxManagerIds :type Array%20of%20Names))
  (bind _localfxmanagerkey (LocalVariables|GetLocalFxManagerKey))
  (bind _stagemanager (Variables|Default|GetStageManager))
  (bind _fxmanagersmap (Variables|/SparksStageShow/Systems/StageManager/BP_MojoStageManagerBase.BP_MojoStageManagerBase_C|GetFxManagersMap _stagemanager))
  (bind (_value _returnvalue) (Utilities|Map|Find _fxmanagersmap _localfxmanagerkey))
  (Utilities|Array|Clear (LocalVariables|GetLocalFxManagers))
  (for _array_element FxManagerIds
    ;[1 0.8712 0.1183] If it is battle stage extract the index of the fxmanager id before execute it.
    (if (Variables|/SparksStageShow/Systems/StageManager/BP_MojoStageManagerBase.BP_MojoStageManagerBase_C|GetIsBattleStage (Variables|Default|GetStageManager))
      (bind _keys (Utilities|Map|Keys (Variables|/SparksStageShow/Systems/StageManager/BP_MojoStageManagerBase.BP_MojoStageManagerBase_C|GetFxManagersMap (Variables|Default|GetStageManager))))
      (for _array_element_2 _keys
        (if (.ReturnValue (Utilities|String|Split (Utilities|String|ToString(Name) _array_element_2) "_" "IgnoreCase" "FromEnd"))
          (if (== _array_element (Utilities|String|StringToName (Utilities|String|Split (Utilities|String|ToString(Name) _array_element_2) "_" "IgnoreCase" "FromEnd")))
            _))
        (LocalVariables|SetLocalFxManagerKey (select (.ReturnValue (Utilities|String|Split (Utilities|String|ToString(Name) _array_element_2) "_" "IgnoreCase" "FromEnd")) _array_element_2 _array_element))
        (if _returnvalue
          (Utilities|IsValid _value
            (:Is%20Valid
              (Utilities|Array|Add (LocalVariables|GetLocalFxManagers) _value)))
          (elif (Variables|/SparksStageShow/Systems/StageManager/BP_MojoStageManagerBase.BP_MojoStageManagerBase_C|GetEnableDebugLogs (Variables|Default|GetStageManager))
            (Development|PrintText (Utilities|Text|FormatText "NSLOCTEXT(\"[4E16E6D6F54585BA172502B426D9D680]\", \"3CD97C534F08CA8C71C3BFBB0FE5107E\", \"[StageEvent] FxManager ID not found on the FxManagersMap: {id}\")" _localfxmanagerkey) true true "(R=1.000000,G=0.082175,B=0.155871,A=1.000000)" 1e+01))))
      (else
        (LocalVariables|SetLocalFxManagerKey _array_element))))
  (return (CallFunction|GetActiveFxManagers (LocalVariables|GetLocalFxManagers))))

(fn PlayEvent
  (bind _fxmanagerevent (Variables|Default|GetFXManagerEvent))
  (if (Utilities|IsValid _fxmanagerevent :!class /Script/Engine.KismetSystemLibrary)
    ;[1 1 1] If PlayOnAllFXManagers is true, we send the event to all FXManagers.
    ; If not, we send only to the selected indices
    (if (Variables|/SparksStageShow/Systems/StageManager/FXManager/MojoFxManagerEvent_DataAsset.MojoFxManagerEvent_DataAsset_C|GetPlayOnAllFXManagers _fxmanagerevent)
      (bind _values (Utilities|Map|Values (Variables|/SparksStageShow/Systems/StageManager/BP_MojoStageManagerBase.BP_MojoStageManagerBase_C|GetFxManagersMap (Variables|Default|GetStageManager))))
      (Variables|Default|SetActiveFXManagers (CallFunction|GetActiveFxManagers _values))
      (else
        (bind _validfxmanagers (CallFunction|GetValidFxManagers (Variables|/SparksStageShow/Systems/StageManager/FXManager/MojoFxManagerEvent_DataAsset.MojoFxManagerEvent_DataAsset_C|GetTargetFxManagerIds _fxmanagerevent)))
        (Variables|Default|SetActiveFXManagers _validfxmanagers)))
    ;[1 1 1] In case there is none fxmanager found just set as event complete
    (if (Utilities|Array|IsNotEmpty (Variables|Default|GetActiveFXManagers))
      ;[1 1 1] Fire the events on the FxManagers and binds CheckIfEventsIsComplete to FxManager's EventFinished dispatcher to be notified when it finished playing the event
      ;;[1 1 1] Local array with the FxManagers that will execute StageShow (Safer than iterating Active array, since elements can be removed when a event finishes at the same frame it's invoked
      (LocalVariables|SetLocalFxManagers (Variables|Default|GetActiveFXManagers))
      (for _array_element (LocalVariables|GetLocalFxManagers)
        (Default|BindEventtoEventFinished _array_element (EventDispatchers|CreateEvent :!event Check%20if%20Event%20Is%20Complete)))
      (Class|BPMojoFxManager|StageShowEvent (LocalVariables|GetLocalFxManagers) (Variables|Default|GetFXManagerEvent))
      (else
        (Default|CallEventComplete self)))))

(fn StartEvent
  (bind _fxmanagerevent (Variables|Default|GetFXManagerEvent))
  (bind _startdelay (Variables|/SparksStageShow/Systems/StageManager/FXManager/MojoFxManagerEvent_DataAsset.MojoFxManagerEvent_DataAsset_C|GetStartDelay _fxmanagerevent))
  (Utilities|IsValid _fxmanagerevent
    (:Is%20Valid
      (if (> _startdelay 0)
        ;[1 1 1] If InDelay is set, wait by the defined time before firing this event
        (bind _returnvalue (Utilities|Time|SetTimerbyEvent (EventDispatchers|CreateEvent :!event PlayEvent) _startdelay))
        (Variables|Default|SetStartDelayTimerHandle _returnvalue)
        (else
          (CallFunction|PlayEvent)))
      (if (Variables|/SparksStageShow/Systems/StageManager/BP_MojoStageManagerBase.BP_MojoStageManagerBase_C|GetEnableDebugLogs (Variables|Default|GetStageManager))
        (Development|PrintText (Utilities|Text|FormatText "NSLOCTEXT(\"[4E16E6D6F54585BA172502B426D9D680]\", \"3A10B9654741554F77845285DA4FDE77\", \"[StageEvent] Start Event: {event}\")" (Utilities|GetDisplayName (Variables|Default|GetFXManagerEvent))) false)))
    (:Is%20Not%20Valid
      (Fortnite|FortShippingLog self (Utilities|String|Append "[StageEvent] Invalid stage event: " (Utilities|GetObjectName _fxmanagerevent)) true))))

(fn ConstructionScript)
