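; Base stage manager actor for the Sparks stage show, coordinating FX managers and playback of stage sequences.
; It maps same-level FX managers by ID and registers itself with them on cosmetic-capable clients.
; MIDI "SECTION" text events become enums for an empty OnMidiEventStart hook, and Niagara overrides from sequence pools are async-loaded.
; PlaySequence stops running sequences, then spawns one BP_MojoStageSequence actor per sequence, destroying each on completion.
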(event EventBeginPlay
  (Networking|CanExecuteCosmeticEvents
    (:True
      (Fortnite|FortShippingLog self "[StageManager] Initializing")
      (bind _values (Utilities|Map|Values (Variables|Default|GetFxManagersMap)))
      (for _array_element _values
        (Utilities|IsValid _array_element
          (:Is%20Valid
            (Variables|/SparksStageShow/Systems/StageManager/FXManager/BP_MojoFxManager.BP_MojoFxManager_C|SetStageManager self _array_element)))))))

(event Custom|OnMidiEventDriverChanged
  (params (MidiEventDriver :type Midi%20Event%20Driver%20Interface%20Interface))
  ;[1 1 1] Start to listen the MIDI Text Event
  (bind (_asynctaskproxy _midieventtext) (MidiEvents|ListenForMidiTextEvents MidiEventDriver "SECTION" (Variables|Default|GetEvent%20Strings))
    (:OnMidiEventStart
      (if (Variables|Default|GetEnableDebugLogs)
        (Development|PrintString (Utilities|String|Append "[StageManager] RECEIVED MIDI TEXT EVENT -> " _midieventtext) true true "(R=0.000000,G=0.660000,B=1.000000,A=1.000000)" 1e+01))
      (CallFunction|OnMidiEventStart (MidiEvents|TextEvents|Events|GetStageManagerEventsMidiTextEnumfromString _midieventtext)))))

(event Custom|LoadEventsNiagaraSystems
  (params (SequencePool :type Array%20of%20Stage%20Show%20Sequence%20Pool%20Object%20References))
  (Utilities|Array|Clear (Variables|Default|GetNiagarasToLoad))
  (for _array_element SequencePool
    (for _array_element_2 (Variables|/Script/SparksStageShowRuntime.StageShowSequencePool|GetSequences _array_element)
      ;[1 1 1] Check each event struct
      (for _array_element_3 (Variables|/Script/SparksStageShowRuntime.StageShowSequence|GetSequenceSteps _array_element_2)
        ;[1 1 1] Check each individual event and add found Niagara System soft references to the array to load
        (for _array_element_4 (Utilities|Struct|BreakStageManagerSequenceSteps _array_element_3)
          (Utilities|IsValid _array_element_4
            (:Is%20Valid
              (if (Utilities|IsValidSoftObjectReference (Variables|/SparksStageShow/Systems/StageManager/FXManager/MojoFxManagerEvent_DataAsset.MojoFxManagerEvent_DataAsset_C|GetNiagaraSystemOverride _array_element_4))
                (Utilities|Array|AddUnique (Variables|Default|GetNiagarasToLoad) (Variables|/SparksStageShow/Systems/StageManager/FXManager/MojoFxManagerEvent_DataAsset.MojoFxManagerEvent_DataAsset_C|GetNiagaraSystemOverride _array_element_4)))))))))
  ;[1 1 1] Load all Niagaras found and stores a reference to the loaded objects on an array
  (Utilities|Array|Clear (Variables|Default|GetLoadedNiagaras))
  (Variables|Default|SetNiagaraIterator 0)
  (if (< (Variables|Default|GetNiagaraIterator) (Utilities|Array|Length (Variables|Default|GetNiagarasToLoad)))
    (bind _object (Utilities|AsyncLoadAsset (Utilities|Array|Get(acopy) (Variables|Default|GetNiagarasToLoad) (Variables|Default|GetNiagaraIterator)))
      (:Completed
        (if (Utilities|IsValid _object :!class /Script/Engine.KismetSystemLibrary)
          (Utilities|Array|AddUnique (Variables|Default|GetLoadedNiagaras) _object)
          (if (not (Variables|Default|GetEnableDebugLogs))
            (Math|Integer|IncrementInt (Variables|Default|GetNiagaraIterator)
              (:%20%20
                _)))
          (Development|PrintText (Utilities|Text|FormatText "NSLOCTEXT(\"[7FDCD01A1A5BE285969AE269C99C0712]\", \"ECE2F71B46636737C55B45BA692564EB\", \"[StageManager] {index}: {name} loaded.\")" (Variables|Default|GetNiagaraIterator) (Utilities|GetDisplayName _object)) false false "(R=0.090404,G=1.000000,B=0.240329,A=1.000000)" 1e+01)))))
  (Development|PrintString "[StageManager] Finished loading assets." false)
  (Default|CallFinishedLoadingAssets))

(fn GrabAllFxManagersFromLevel
  (Utilities|Map|Clear (Variables|Default|GetFxManagersMap))
  ;[1 1 1] Get all FxManagers on the level and iterate
  (bind _outactors (Actor|GetAllActorsOfClass "/SparksStageShow/Systems/StageManager/FXManager/BP_MojoFxManager.BP_MojoFxManager_C"))
  (for _array_element _outactors
    (if (== (Level|GetLevel _array_element) (Level|GetLevel))
      ;[1 1 1] If a FxManager Id property is empty, add it's display name as key as a fallback
      (if (== "" (Variables|/SparksStageShow/Systems/StageManager/FXManager/BP_MojoFxManager.BP_MojoFxManager_C|GetFxManagerId _array_element))
        (LocalVariables|SetId (Utilities|String|StringToName (Utilities|GetDisplayName _array_element)))
        (else
          (LocalVariables|SetId (Variables|/SparksStageShow/Systems/StageManager/FXManager/BP_MojoFxManager.BP_MojoFxManager_C|GetFxManagerId _array_element))))
      ;[1 1 1] Check if it's a unique Id before adding to the Map
      (if (not (Utilities|Map|Contains (Variables|Default|GetFxManagersMap) (LocalVariables|GetId)))
        (Utilities|Map|Add (Variables|Default|GetFxManagersMap) (LocalVariables|GetId) _array_element)))))

(fn MapMidiTextEventDataTable
  (params (MidiEvents :type Array%20of%20ESparksStageManager_Events_Midi_Text_Events%20Enums :default "NewEnumerator0"))
  ;[1 1 1] Get all mapped events from a MIDI Text Event Data Table
  (Utilities|Array|Clear (Variables|Default|GetEvent%20Strings))
  (for _array_element MidiEvents
    (Utilities|Array|AddUnique (Variables|Default|GetEvent%20Strings) (MidiEvents|TextEvents|Events|GetStageManagerEventsMidiTextStringfromEnum _array_element))))

(fn OnMidiEventStart
  (params (MidiTextEvent :type ESparksStageManager_Events_Midi_Text_Events%20Enum%20%28by%20ref%29 :default "NewEnumerator0")))

(fn PlaySequence
  (params (Sequences :type Array%20of%20Stage%20Show%20Sequence%20Object%20References))
  ; Stop any sequences running
  (Default|StopSequences)
  ;[1 1 1] Grab the sequence array and start to spawn the MojoStageSequence Actors
  (for _array_element Sequences
    (bind _asmojo_fx_manager_sequence_data_asset (Utilities|Casting|CastToMojoFxManagerSequence_DataAsset _array_element))
    ;[1 1 1] Spawn the actor and feed with the necessary data
    (bind _returnvalue (Game|SpawnActorfromClass "/SparksStageShow/Systems/StageManager/BP_MojoStageSequence.BP_MojoStageSequence_C" (Transformation|GetActorTransform) "Undefined" "OverrideRootScale" 0 _asmojo_fx_manager_sequence_data_asset self 0))
    (Utilities|IsValid _returnvalue
      (:Is%20Valid
        (Default|BindEventtoSequenceComplete _returnvalue (EventDispatchers|CreateEvent :!event Sequence%20Actor%20Cleanup)) ; When finishes Call Destroy
        (Utilities|Array|Add (Variables|Default|GetActiveSequences) _returnvalue)
        (Class|BPMojoStageSequence|PlaySequence _returnvalue)))))

(fn SequenceActorCleanup
  (params (Caller :type BP%20Mojo%20Stage%20Sequence%20Object%20Reference))
  (bind _activesequences (Variables|Default|GetActiveSequences))
  (bind _returnvalue (Utilities|Array|Length _activesequences))
  (bind _returnvalue_2 (Utilities|GetDisplayName Caller))
  (Utilities|IsValid Caller
    (:Is%20Valid
      (Default|UnbindEventfromSequenceComplete Caller (EventDispatchers|CreateEvent :!event Sequence%20Actor%20Cleanup))
      (bind _returnvalue_3 (Utilities|Array|RemoveItem (Variables|Default|GetActiveSequences) Caller))
      (if _returnvalue_3
        (Actor|DestroyActor Caller)
        (if (Variables|Default|GetEnableDebugLogs)
          (Development|PrintText (Utilities|Text|FormatText "NSLOCTEXT(\"[7FDCD01A1A5BE285969AE269C99C0712]\", \"600992B5499270B93B4422929CB26693\", \"[StageManager] Sequence {name} completed. Still active: {active}\")" _returnvalue_2 _returnvalue) false))
        (else
          (if (Variables|Default|GetEnableDebugLogs)
            (Development|PrintText (Utilities|Text|FormatText "NSLOCTEXT(\"[7FDCD01A1A5BE285969AE269C99C0712]\", \"714D6FDA4D4866D96344DDA56BC4817F\", \"[StageManager] Error: Sequence {name} not found on active list on cleanup. Still active: {active}\")" _returnvalue_2 _returnvalue) false)))))))

(fn StopSequences
  (bind _activesequences (Variables|Default|GetActiveSequences))
  (for _array_element _activesequences
    (Utilities|IsValid _array_element
      (:Is%20Valid
        (Default|UnbindEventfromSequenceComplete _array_element (EventDispatchers|CreateEvent :!event Sequence%20Actor%20Cleanup))
        (Class|BPMojoStageSequence|DestroySequence _array_element))))
  (Utilities|Array|Clear _activesequences)
  (bind _values (Utilities|Map|Values (Variables|Default|GetFxManagersMap)))
  (for _array_element_2 _values
    (Class|BPMojoFxManager|StopEventandFinish _array_element_2)))

(fn ConstructionScript)
