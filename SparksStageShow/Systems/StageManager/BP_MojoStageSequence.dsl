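; Stage sequence actor that plays a StageShowSequence step by step for the stage manager.
; PlaySequence spawns all BP_MojoStageEvent actors in a step before starting them, and skips empty steps.
; Completed events are destroyed; once none remain active, it recurses to the next step, firing SequenceComplete past the last.
; DestroySequence tears down active events; ValidateUniqueEventManagers, uncalled here, filters events with duplicate FX manager IDs.
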(fn CheckifAllEventsComplete
  (params (Caller :type BP%20Mojo%20Stage%20Event%20Object%20Reference))
  (LocalVariables|SetLocalCaller Caller)
  (Utilities|FlowControl|Sequence
    (:then_0
      ;[1 0.6208 0.7531] Just for debug purpose
      (if (Variables|/SparksStageShow/Systems/StageManager/BP_MojoStageManagerBase.BP_MojoStageManagerBase_C|GetEnableDebugLogs (Variables|Default|GetStageManager))
        (Development|PrintString (Utilities|String|Append "[StageEvent] Event ended: " (Utilities|GetDisplayName (LocalVariables|GetLocalCaller)) ". Active events: " (Utilities|String|ToString(Integer) (Utilities|Array|Length (Variables|Default|GetActiveEvents)))) false)))
    (:then_1
      ;[1 1 1] Check if exist on active list if exist destroy and remove if all active events finish play the next event list
      (if (>= (Utilities|Array|FindItem (Variables|Default|GetActiveEvents) (LocalVariables|GetLocalCaller)) 0)
        (bind _returnvalue (Utilities|Array|RemoveItem (Variables|Default|GetActiveEvents) (LocalVariables|GetLocalCaller)))
        (Actor|DestroyActor (LocalVariables|GetLocalCaller))
        (if _returnvalue
          (if (Utilities|Array|IsEmpty (Variables|Default|GetActiveEvents))
            (Math|Integer|IncrementInt (Variables|Default|GetCounter)
              (:%20%20
                (CallFunction|PlaySequence)))))))))

(fn DestroySequence
  (for _array_element (Variables|Default|GetActiveEvents)
    (Utilities|IsValid _array_element
      (:Is%20Valid
        (Default|UnbindEventfromEventComplete _array_element (EventDispatchers|CreateEvent :!event Check%20if%20All%20Events%20Complete))
        (Class|BPMojoStageEvent|DestroyEvent _array_element))))
  (Actor|DestroyActor))

(fn PlaySequence
  (bind _counter (Variables|Default|GetCounter))
  (bind _sequence (Variables|Default|GetSequence))
  (bind _sequencesteps (Variables|/Script/SparksStageShowRuntime.StageShowSequence|GetSequenceSteps _sequence))
  (bind _output (Utilities|Array|Get(acopy) _sequencesteps _counter))
  (bind _stepevents (Utilities|Struct|BreakStageManagerSequenceSteps _output))
  ;[0.4949 0.4643 1] Prevent extrapolate the Fx length value
  (if (< (Variables|Default|GetCounter) (Utilities|Array|Length (Variables|/Script/SparksStageShowRuntime.StageShowSequence|GetSequenceSteps (Variables|Default|GetSequence))))
    ;[1 0.9454 0.4708] Check if this list has unique fxmanager indices if find the same indice return false and go to the next list of event
    (if (Utilities|Array|IsNotEmpty _stepevents)
      ;[0.146 1 0.146] If Everything is ok spawn a stage event and start it
      (for _array_element _stepevents
        (bind _asmojo_fx_manager_event_data_asset (Utilities|Casting|CastToMojoFxManagerEvent_DataAsset _array_element))
        (bind _returnvalue (Game|SpawnActorfromClass "/SparksStageShow/Systems/StageManager/BP_MojoStageEvent.BP_MojoStageEvent_C" (Transformation|GetActorTransform) "Undefined" "OverrideRootScale" 0 (Variables|Default|GetStageManager) _asmojo_fx_manager_event_data_asset 0))
        ;[1 0.8625 0.1117] Register this event as active and also bind to event complete to update the active event list and start another list of events.
        (Utilities|Array|Add (Variables|Default|GetActiveEvents) _returnvalue)
        (Utilities|Array|Add (LocalVariables|GetEventsToStart) _returnvalue)
        (Default|BindEventtoEventComplete _returnvalue (EventDispatchers|CreateEvent :!event Check%20if%20All%20Events%20Complete)))
      (for _array_element_2 (LocalVariables|GetEventsToStart)
        (Class|BPMojoStageEvent|StartEvent _array_element_2))
      (else
        (Math|Integer|IncrementInt (Variables|Default|GetCounter)
          (:%20%20
            (CallFunction|PlaySequence)))))
    (else
      ;[0.5144 1 0.5581] If extrapolate reset the counter and clear the active event list and call the sequence complete to let the stage manager know that it can destroy this actor.
      (Variables|Default|SetCounter 0)
      (Utilities|Array|Clear (Variables|Default|GetActiveEvents))
      (Default|CallSequenceComplete self))))

(fn ConstructionScript)

(fn ValidateUniqueEventManagers
  (params (FxEvents :type Array%20of%20Mojo%20Fx%20Manager%20Event%20Data%20Asset%20Object%20References))
  ;[1 1 1] Check all FxManagerIds on all FxEvents DAs received.
  ; We'll build a array with just the valid events. To be considered valid, the targeted FxManagers can't be repeated between events, or else a FxManager would receive more than one event at the same time.
  (Utilities|Array|Clear (LocalVariables|GetLocalValidFXEvents))
  (for _array_element FxEvents
    (bind _output_get (LocalVariables|SetCurrentFxEvent _array_element))
    (Utilities|IsValid _output_get
      (:Is%20Valid
        ;[1 1 1] The FxManagerId is already in the LocalFxManagerIds? If not, add to the array.
        ; If it is there already, it means two different events are targeting the same FxManager,
        ; making the event invalid.
        (for _array_element_2 (Variables|/SparksStageShow/Systems/StageManager/FXManager/MojoFxManagerEvent_DataAsset.MojoFxManagerEvent_DataAsset_C|GetTargetFxManagerIds _output_get)
          (if (> (Utilities|Array|FindItem (LocalVariables|GetLocalFxManagerIds) _array_element_2) 0)
            (LocalVariables|SetIsValid false)
            ;[0.8796 0.7394 1] Debug purpose to let the user know which event is invalid and where
            (if (Variables|/SparksStageShow/Systems/StageManager/BP_MojoStageManagerBase.BP_MojoStageManagerBase_C|GetEnableDebugLogs (Variables|Default|GetStageManager))
              (Development|PrintString (Utilities|String|ToString(Text) (Utilities|Text|FormatText "NSLOCTEXT(\"[30957D416020D3FB89DA32D597367585]\", \"5392030D494A2E01D0F6FE9B577034FB\", \"[StageSequence] Found duplicated FX Manager indice on {fxeventname}. Duplicated indice -> {idValue} from {stageFXDA} - Skipping >>>>\")" (Utilities|GetDisplayName (LocalVariables|GetCurrentFxEvent)) _array_element_2 (Utilities|GetObjectName (Variables|Default|GetSequence)))) true true "(R=0.791667,G=0.000000,B=0.011117,A=1.000000)" 1e+01))
            (else
              (Utilities|Array|Add (LocalVariables|GetLocalFxManagerIds) _array_element_2))))
        ;[1 1 1] If the event passed validation, it's added to valid events array.
        (if (LocalVariables|GetIsValid)
          (Utilities|Array|Add (LocalVariables|GetLocalValidFXEvents) (LocalVariables|GetCurrentFxEvent))))))
  ;[1 1 1] Returns just the events that passed validation
  (return (LocalVariables|GetLocalValidFXEvents) (> (Utilities|Array|Length (LocalVariables|GetLocalValidFXEvents)) 0)))
