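# Masked, lit material for 3D editor text, apparently for StageManager labels, with nodes to turn the text toward the camera.
# A Custom atan2 node turns the normalized object-to-camera direction into a yaw, scaled by 2π into turns.
# RotateAboutAxis applies it to world position around object orientation, with a UV-derived pivot transformed to world space.
# The Roboto distance-field font's red channel is remapped through a steep lerp for sharp edges, alongside a vertex color node.
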Material /SparksStageShow/Systems/StageManager/M_EditorTextCameraFace
graph {
 properties {blendMode=BLEND_Masked, pixelDepthOffsetMode=PDOM_Legacy, usedShadingModels=Default Lit}
 nodes [
  LinearInterpolate(Alpha=FontSampleParameter() {font=/Engine/EngineFonts/RobotoDistanceField, parameterName=Font}.R) {constA=-90, constB=90}
  MaterialExpressionNormalize_1 = Normalize(VectorInput=Subtract(A=CameraPositionWS(), B=ObjectPositionWS()))
  n1 = Custom!(In=@MaterialExpressionNormalize_1) {
   code="float2 output; float2 Input1; output = atan2 (In.y,In.x); return(output);"
   outputType=CMOT_Float1
   inputs [{inputName=In, input {expression=&self:MaterialExpressionNormalize_1}}]
  }
  n2 = Constant() {r=3.142}
  n3 = TextureCoordinate()
  n4 = ComponentMask(None=TextureCoordinate()) {r=true}
  RotateAboutAxis(
   NormalizedRotationAxis=ObjectOrientation()
   RotationAngle=Divide(A=Divide(A=@n1, B=@n2), B=Constant() {r=2})
   PivotPoint=TransformPosition(None=AppendVector(A=@n3, B=@n4)) {transformType=TRANSFORMPOSSOURCE_World}
   Position=WorldPosition().XYZ
  )
  VertexColor()
 ]
}
