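; Blueprint function library for the Sparks stage show's Stage Manager that converts between string names and enums for MIDI text events and quickplay (QP) events.
; Two data tables act as the lookup, with row names as the strings and a struct field holding each enum.
; String-to-enum reads the row directly; enum-to-string scans every row for a match.
; Other functions list all row names, and failed row lookups print an error and return fallback values.
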(fn GetAllStageManagerEventsMidiTextStrings
  (bind _outrownames (DataTable|GetDataTableRowNames "/SparksStageShow/Systems/StageManager/MIDITextEvents/DT_SparksStageManagerEventsMidiTextEvents.DT_SparksStageManagerEventsMidiTextEvents"))
  (for _array_element _outrownames
    (Utilities|Array|AddUnique (LocalVariables|GetMidiTextEvents) (Utilities|String|ToString(Name) _array_element)))
  ;[1 0.5 0.5] Return all values found in the table
  (return (LocalVariables|GetMidiTextEvents)))

(fn GetAllStageManagerQPEventStrings
  (bind _outrownames (DataTable|GetDataTableRowNames "/SparksStageShow/Systems/StageManager/MIDITextEvents/DT_SparksStageManagerQPEvents.DT_SparksStageManagerQPEvents"))
  (for _array_element _outrownames
    (Utilities|Array|AddUnique (LocalVariables|GetQPEvents) (Utilities|String|ToString(Name) _array_element)))
  ;[1 0.5 0.5] Return all values found in the table
  (return (LocalVariables|GetQPEvents)))

(fn GetStageManageQPEventsEnumFromString
  (params (QPEventString :type String))
  (bind _returnvalue (Utilities|GetDataTableRow "/SparksStageShow/Systems/StageManager/MIDITextEvents/DT_SparksStageManagerQPEvents.DT_SparksStageManagerQPEvents" (Utilities|String|StringToName QPEventString))
    (:then
      ;[1 0.5 0.5] Return
      (bind _returnvalue_qpevent_22_d3502fde4790d892d2366f8e91ffd0ad (Utilities|Struct|BreakSSparksStageManagerQEventsTableRow _returnvalue))
      (return _returnvalue_qpevent_22_d3502fde4790d892d2366f8e91ffd0ad))
    (:RowNotFound
      ;[1 0 0] Error: Could not find row
      (Development|PrintString 0 (Utilities|String|Append "Could not find row for QP Event: " (Variables|GetQPEventString)) true true "(R=0.000000,G=0.660000,B=1.000000,A=1.000000)" 5.0)
      (return "Invalid"))))

(fn GetStageManagerEventsMidiTextEnumFromString
  (params (MidiTextEventString :type String))
  (bind _returnvalue (Utilities|GetDataTableRow "/SparksStageShow/Systems/StageManager/MIDITextEvents/DT_SparksStageManagerEventsMidiTextEvents.DT_SparksStageManagerEventsMidiTextEvents" (Utilities|String|StringToName MidiTextEventString))
    (:then
      ;[1 0.5 0.5] Return
      (bind _returnvalue_miditextevent_19_d3502fde4790d892d2366f8e91ffd0ad (Utilities|Struct|BreakSSparksStageManagerEventsMidiTextEventsTableRow _returnvalue))
      (return _returnvalue_miditextevent_19_d3502fde4790d892d2366f8e91ffd0ad))
    (:RowNotFound
      ;[1 0 0] Error: Could not find row
      (Development|PrintString 0 (Utilities|String|Append "Could not find row for Midi Text: " (Variables|GetMidiTextEventString)) true true "(R=0.000000,G=0.660000,B=1.000000,A=1.000000)" 5.0)
      (return "NewEnumerator0"))))

(fn GetStageManagerEventsMidiTextStringFromEnum
  (params (MidiTextEventEnum :type ESparksStageManager_Events_Midi_Text_Events%20Enum :default "NewEnumerator0"))
  (bind _outrownames (DataTable|GetDataTableRowNames "/SparksStageShow/Systems/StageManager/MIDITextEvents/DT_SparksStageManagerEventsMidiTextEvents.DT_SparksStageManagerEventsMidiTextEvents"))
  (for _array_element _outrownames
    (bind _returnvalue (Utilities|GetDataTableRow "/SparksStageShow/Systems/StageManager/MIDITextEvents/DT_SparksStageManagerEventsMidiTextEvents.DT_SparksStageManagerEventsMidiTextEvents" _array_element)
      (:then
        (bind _returnvalue_miditextevent_19_d3502fde4790d892d2366f8e91ffd0ad (Utilities|Struct|BreakSSparksStageManagerEventsMidiTextEventsTableRow _returnvalue))
        (if (Utilities|Enum|Equal(Enum) _returnvalue_miditextevent_19_d3502fde4790d892d2366f8e91ffd0ad MidiTextEventEnum)
          ;[1 0.5 0.5] Return the corresponding int
          (return (Utilities|String|ToString(Name) _array_element))))
      (:RowNotFound
        ;[1 0 0] Error: Could not find row
        (Development|PrintString 0 (Utilities|String|Append "Could not find row for Midi Text: " (Utilities|String|EnumtoString MidiTextEventEnum)) true true "(R=0.000000,G=0.660000,B=1.000000,A=1.000000)" 5.0)
        (return "")))))

(fn GetStageManagerQPEventstStringFromEnum
  (params (QPEventEnum :type EPilgrimQuickplayState%20Enum :default "Invalid"))
  (bind _outrownames (DataTable|GetDataTableRowNames "/SparksStageShow/Systems/StageManager/MIDITextEvents/DT_SparksStageManagerQPEvents.DT_SparksStageManagerQPEvents"))
  (for _array_element _outrownames
    (bind _returnvalue (Utilities|GetDataTableRow "/SparksStageShow/Systems/StageManager/MIDITextEvents/DT_SparksStageManagerQPEvents.DT_SparksStageManagerQPEvents" _array_element)
      (:then
        (bind _returnvalue_qpevent_22_d3502fde4790d892d2366f8e91ffd0ad (Utilities|Struct|BreakSSparksStageManagerQEventsTableRow _returnvalue))
        (if (Utilities|Enum|Equal(Enum) _returnvalue_qpevent_22_d3502fde4790d892d2366f8e91ffd0ad QPEventEnum)
          ;[1 0.5 0.5] Return the corresponding int
          (return (Utilities|String|ToString(Name) _array_element))))
      (:RowNotFound
        ;[1 0 0] Error: Could not find row
        (Development|PrintString 0 (Utilities|String|Append "Could not find row for QP Text: " (Utilities|String|EnumtoString QPEventEnum)) true true "(R=0.000000,G=0.660000,B=1.000000,A=1.000000)" 5.0)
        (return "")))))
